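# Unlit, translucent, two-sided Niagara material for MistRaven's dark coil effect, using SplineThicken to widen thin geometry.
# Two panning noise samples drive a curve atlas color ramp, tinted by particle color for emissive.
# The noise, masked at UV borders, also shapes opacity with depth and view-angle fades.
# Dynamic parameters control width, panning, emissive and world position offset terms for waves, bobbing and velocity-driven lag, with optional vertex-alpha anchoring.
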Material /Game/Effects/Niagara/MasterMaterials/MistRaven/M_MistRaven_DarkCoil_SplineThicken
graph {
 properties {
  bAutomaticallySetUsageInEditor=false
  bUsedWithNiagaraMeshParticles=true
  bUsedWithNiagaraRibbons=true
  bUsedWithNiagaraSprites=true
  blendMode=BLEND_Translucent
  shadingModel=MSM_Unlit
  twoSided=true
  usedShadingModels=Unlit
 }
 nodes [
  n1 = OneMinus(None=Subtract(A=NamedRerouteUsage(), B=NamedRerouteUsage()))
  n2 = Multiply(B=LinearInterpolate(Alpha=Clamp(None=@n1)) {constA=1, constB=-1}) {constA=1}
  n3 = Add(A=@n2, B=NamedRerouteUsage())
  n4 = DynamicParameter() {parameterIndex=1, defaultValue=<rgba 1 1 -0.25 1>, paramNames [WidthBase,WidthTip,PanSpeed,Emissive]}
  n5 = Add(A=OneMinus(None=NamedRerouteUsage()), B=NamedRerouteUsage()) {constB=0.5}
  Add(
   A=Add(A=@n3, B=NamedRerouteUsage())
   B=MaterialFunctionCall(
    WidthBase=Multiply(A=@n4.WidthBase, B=@n5)
    WidthTip=Multiply(A=@n4.WidthTip, B=@n5)
    DeriveNormalZ=StaticBool()
    AngleCorrectedNormals=StaticBool()
    AdditionalNormal=VertexNormalWS()
    FlattenPixelNormal=StaticBool() {value=true}
    UVs For NormalShading=TextureCoordinate()
   ) {materialFunction=&SplineThicken}.WorldPositionOffset
  )
  Add(A=@n3, B=NamedRerouteUsage())
  n6 = MaterialFunctionCall(X=Constant(), Y=Add(A=NamedRerouteUsage(), B=NamedRerouteUsage())) {materialFunction=&MakeFloat2}
  n7 = Panner(Coordinate=TextureCoordinate(), Speed=@n6.Result) {speedY=-0.25}
  n8 = TextureObject() {texture=/Game/Global/Textures/Default/Noise/T_StringyEnergy_Noise}
  n9 = Panner(Coordinate=TextureCoordinate() {uTiling=0.5}, Speed=@n6.Result) {speedY=-0.25}
  n10 = Power(
   Base=LinearInterpolate(
    A=TextureSample(UVs=@n7, Tex=@n8) {texture=/Game/Effects/FogVolume/T_TileNoise}.R
    B=TextureSample(UVs=@n9, Tex=@n8) {texture=&T_TileNoise}.B
   )
   Exp=Constant() {r=1}
  ) {constExponent=4}
  n11 = CurveAtlasRowParameter(CurveTime=@n10) {
   atlas=&/Debug_CurveAtlas
   curve=&/CC_MistRaven
   parameterName=CC
  }
  n12 = ParticleColor()
  MaterialFunctionCall(Emissive=Multiply(A=NamedRerouteUsage(), B=Multiply(A=@n11, B=@n12.RGB))) {materialFunction=&MF_FX_EmissiveAdjustments}
  n13 = Time()
  Multiply(A=@n13)
  Multiply(A=@n2)
  n14 = MaterialFunctionCall(Float2=TextureCoordinate()) {materialFunction=&BreakOutFloat2Components}
  n15 = Power(Base=Multiply(A=@n14.G, B=OneMinus(None=@n14.G))) {constExponent=1.5}
  n16 = Multiply(A=@n10, B=Saturate(None=Multiply(A=@n15) {constB=15}))
  n17 = Multiply(A=@n14.R, B=OneMinus(None=@n14.R))
  n18 = Saturate(None=Multiply(A=Power(Base=@n17)) {constB=15})
  n19 = Multiply(A=@n12.A, B=SmoothStep(Value=Multiply(A=@n16, B=@n18)) {constMax=0.6})
  Multiply(A=@n19, B=@n2)
  n20 = Multiply(A=Time(), B=NamedRerouteUsage())
  n21 = NamedRerouteUsage()
  n22 = Add(A=@n20, B=@n21)
  n23 = Sine(None=@n22)
  n24 = Cosine(None=@n22) {period=2}
  n25 = Subtract(A=@n20, B=@n21)
  n26 = Cosine(None=@n25)
  n27 = Constant() {r=1}
  n28 = NamedRerouteUsage()
  n29 = Multiply(
   A=Multiply(
    A=MaterialFunctionCall(X=@n23, Y=@n24, Z=@n26) {materialFunction=&MakeFloat3}.Result
    B=StaticSwitchParameter(True=@n27, False=@n28) {defaultValue=true, parameterName=DisableVertexMaskAnchor}
   )
   B=NamedRerouteUsage()
  )
  Multiply(A=@n29, B=VertexNormalWS())
  Multiply(
   A=Multiply(A=@n19, B=DepthFade() {fadeDistanceDefault=1})
   B=MaterialFunctionCall("FadeSensitivity (optional)"=ScalarParameter() {defaultValue=0.5, parameterName=ViewAngleFade}) {materialFunction=&FadeBasedOnViewAngle}.Result
  )
  n30 = VertexColor()
  NamedRerouteDeclaration(None=@n30) {name=VertexColor, nodeColor=<rgba 0.0781 0 1 1>}
  n31 = DynamicParameter() {defaultValue=<rgba 0 0 2 1>, paramNames [WPO_Wiggle,WPO_Bounce,WPO_Growth,TimeMult]}
  NamedRerouteDeclaration(None=@n31.WPO_Wiggle) {name=WPO_0, nodeColor=<rgba 0.8047 1 0 1>}
  NamedRerouteDeclaration(None=ComponentMask(None=@n30.A) {r=true}) {name=Vertex Mask, nodeColor=<rgba 0.3594 0 1 1>}
  n32 = TransformPosition(None=WorldPosition().XYZ) {transformSourceType=TRANSFORMPOSSOURCE_World}
  NamedRerouteDeclaration(None=Noise!(World Position=@n32, FilterWidth) {bTurbulence=false, levelScale=1, levels=1, noiseFunction=NOISEFUNCTION_ValueALU, scale=0.0125}) {name=Noise, nodeColor=<rgba 1 0 0.2109 1>}
  NamedRerouteDeclaration(None=@n31.WPO_Bounce) {name=WPO_1, nodeColor=<rgba 0.875 0 1 1>}
  NamedRerouteDeclaration(None=Multiply(A=@n32, B=@n29)) {name=Waves_WPO, nodeColor=<rgba 0 1 0.8359 1>}
  NamedRerouteDeclaration(None=@n31.WPO_Growth) {name=WPO_2, nodeColor=<rgba 0 1 0.5547 1>}
  NamedRerouteDeclaration(None=@n31.TimeMult) {name=TimeMult, nodeColor=<rgba 1 0.2344 0 1>}
  n33 = Constant()
  n34 = NamedRerouteUsage()
  n35 = MaterialFunctionCall(X=@n33, Y=@n33, Z=Sine(None=Add(A=@n13, B=@n34))) {materialFunction=&MakeFloat3}
  n36 = StaticSwitchParameter(True=Constant() {r=1}, False=NamedRerouteUsage()) {defaultValue=true, parameterName=DisableVertexMaskAnchor}
  NamedRerouteDeclaration(None=Multiply(A=Multiply(A=@n35.Result, B=@n36) {constB=25}, B=NamedRerouteUsage())) {name=Bobbing_WPO, nodeColor=<rgba 0 0.0391 1 1>}
  n37 = DynamicParameter() {parameterIndex=2, defaultValue=<rgba 1 0.0125 1 0.0125>, paramNames [VelX,VelY,VelZ,VelInfluence]}
  n38 = Power(Base=NamedRerouteUsage(), Exp=ScalarParameter() {defaultValue=2, group=WPO, parameterName=VertexMask Pow})
  n39 = Multiply(B=@n37.RGB) {constA=-1}
  NamedRerouteDeclaration(None=Multiply(A=@n37.VelInfluence, B=Multiply(A=@n38, B=@n39)) {constA=0.0125}) {name=CharacterMovement, nodeColor=<rgba 0 1 0.0156 1>}
  NamedRerouteDeclaration(None=@n4.PanSpeed) {name=PanSpeed, nodeColor=<rgba 0 0.8828 1 1>}
  NamedRerouteDeclaration(None=@n4.Emissive) {name=Emissive, nodeColor=<rgba 0 1 0.1562 1>}
  NamedRerouteUsage()
  NamedRerouteUsage()
  NamedRerouteUsage()
  Sine(None=@n25)
 ]
}
